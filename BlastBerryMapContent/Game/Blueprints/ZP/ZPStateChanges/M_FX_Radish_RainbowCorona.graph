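# Additive, two-sided unlit particle material for the Radish rainbow corona effect, with a rainbow gradient tinted by particle color feeding emissive adjustments.
# Two counter-panning samples of a light-cone texture multiply into waves, modulated by scrolling soft noise and particle alpha.
# The resulting mask gets a depth fade, which a feature-level switch replaces with constant 1 on ES2/ES3.1.
# A SplineThicken function takes its base and tip width from a dynamic parameter.
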Material /BlastBerryMapContent/Game/Blueprints/ZP/ZPStateChanges/M_FX_Radish_RainbowCorona
graph {
 properties {
  bUsedWithNiagaraMeshParticles=true
  bUsedWithNiagaraSprites=true
  bUsedWithParticleSprites=true
  blendMode=BLEND_Additive
  pixelDepthOffsetMode=PDOM_Legacy
  shadingModel=MSM_Unlit
  twoSided=true
  usedShadingModels=Unlit
 }
 nodes [
  n1 = ParticleColor()
  n2 = MaterialFunctionCall(XY=TextureCoordinate()) {materialFunction=&Swizzle}
  n3 = AppendVector(A=ScalarParameter() {defaultValue=3, parameterName=Waves_Tile_X}, B=Constant() {r=1})
  n4 = Multiply(A=ScalarParameter() {defaultValue=0.025, parameterName=Waves_Pan_X}) {constB=-1}
  n5 = Constant()
  n6 = TextureSample(UVs=Panner(Coordinate=Multiply(A=@n2.YX, B=@n3), Speed=AppendVector(A=@n4, B=@n5))) {texture=&/T_Radish_Coronamesh_lightcone}
  n7 = MaterialFunctionCall(XY=TextureCoordinate()) {materialFunction=&Swizzle}
  n8 = AppendVector(A=ScalarParameter() {defaultValue=3, parameterName=Waves_Tile_X}, B=Constant() {r=1})
  n9 = ScalarParameter() {defaultValue=0.025, parameterName=Waves_Pan_X}
  n10 = Constant()
  n11 = TextureSample(UVs=Panner(Coordinate=Multiply(A=@n7.YX, B=@n8), Speed=AppendVector(A=@n9, B=@n10))) {texture=&T_Radish_Coronamesh_lightcone}
  n12 = Multiply(A=Clamp(None=Multiply(A=@n6.R, B=@n11.R)))
  n13 = Multiply(A=TextureCoordinate(), B=AppendVector(A=Constant() {r=1}, B=Constant() {r=1}))
  n14 = AppendVector(A=Constant(), B=Constant() {r=0.1})
  n15 = LinearInterpolate(
   A=Constant() {r=0.5}
   B=Constant() {r=1}
   Alpha=TextureSample(UVs=Panner(Coordinate=@n13, Speed=@n14)) {texture=/Game/Characters/Player/Male/Medium/Bodies/M_MED_Viper/Textures/Glitch/T_SoftNoise}.RGB
  )
  n16 = DepthFade(Opacity=Multiply(A=@n1.A, B=Multiply(A=@n12, B=@n15)), FadeDistance=ScalarParameter() {defaultValue=100, parameterName=DepthFade})
  n17 = Constant() {r=1}
  FeatureLevelSwitch(Default=@n16, ES2=@n17, ES3_1=@n17, SM4_REMOVED=@n16, SM5=@n16, SM6=@n16)
  n18 = Multiply(
   A=MaterialFunctionCall(XY=TextureCoordinate()) {materialFunction=&Swizzle}.YX
   B=AppendVector(A=Constant() {r=1}, B=Constant() {r=1})
  )
  MaterialFunctionCall(Emissive=Multiply(A=TextureSample(UVs=@n18) {texture=/Game/Effects/Fort_Effects/Textures/Grad/T_Gradient_Rainbow_NoRed}.RGB, B=@n1.RGB)) {materialFunction=&MF_FX_EmissiveAdjustments}
  n19 = DynamicParameter() {defaultValue=<rgba 0.6 1 1 1>, paramNames [SplineThickness, 200, 3, 4]}
  MaterialFunctionCall(WidthBase=@n19.SplineThickness, WidthTip=@n19.SplineThickness) {materialFunction=&SplineThicken}
  n20 = TextureCoordinate()
  Power(Base=Sine(None=ComponentMask(None=@n20) {g=true}) {period=2}) {constExponent=2.2}
 ]
}
