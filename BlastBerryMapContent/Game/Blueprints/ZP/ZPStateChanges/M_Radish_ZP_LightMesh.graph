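# Translucent, unlit, two-sided Niagara mesh-particle material for a light mesh effect in the Radish ZP state changes.
# Emissive scales particle color along a vertical UV falloff through MF_FX_EmissiveAdjustments, while SplineThicken takes width from a dynamic parameter.
# Opacity math combines particle alpha, two opposing-panned soft noise samples, a sine profile across U and V-gradient fades.
# Depth fade softens intersections, with a feature-level switch bypassing it on ES2 and ES3_1.
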Material /BlastBerryMapContent/Game/Blueprints/ZP/ZPStateChanges/M_Radish_ZP_LightMesh
graph {
 properties {
  bUsedWithNiagaraMeshParticles=true
  blendMode=BLEND_Translucent
  pixelDepthOffsetMode=PDOM_Legacy
  shadingModel=MSM_Unlit
  twoSided=true
  usedShadingModels=Unlit
 }
 nodes [
  Constant3Vector() {constant=<rgba 1 0 0.0033 0>}
  Constant3Vector() {constant=<rgba 0.0623 1 0 0>}
  LinearInterpolate(A=Constant() {r=50}, B=Constant() {r=150}, Alpha=ParticleRandom())
  n1 = DynamicParameter() {paramNames [Thickness Start, Thickness End, Param3, Param4]}
  MaterialFunctionCall(WidthBase=@n1.Thickness Start, WidthTip=@n1.Thickness Start) {materialFunction=&SplineThicken}
  n2 = Constant() {r=1}
  n3 = Constant() {r=6}
  n4 = ComponentMask(None=TextureCoordinate()) {g=true}
  n5 = Clamp(None=Power(Base=OneMinus(None=@n4)) {constExponent=2.2})
  n6 = ParticleColor()
  MaterialFunctionCall(Emissive=Multiply(A=LinearInterpolate(A=@n2, B=@n3, Alpha=@n5), B=@n6.RGB)) {materialFunction=&MF_FX_EmissiveAdjustments}
  n7 = TextureCoordinate()
  n8 = AppendVector(A=Constant() {r=1}, B=Constant() {r=1})
  n9 = Constant() {r=0.1}
  n10 = Constant() {r=-0.2}
  n11 = TextureSample(UVs=Panner(Coordinate=Multiply(A=@n7, B=@n8), Speed=AppendVector(A=@n9, B=@n10))) {texture=/Game/Characters/Player/Male/Medium/Bodies/M_MED_Viper/Textures/Glitch/T_SoftNoise}
  n12 = TextureCoordinate()
  n13 = AppendVector(A=Constant() {r=1}, B=Constant() {r=1})
  n14 = Constant() {r=-0.1}
  n15 = Constant() {r=-0.1}
  n16 = TextureSample(UVs=Panner(Coordinate=Multiply(A=@n12, B=@n13), Speed=AppendVector(A=@n14, B=@n15))) {texture=&T_SoftNoise}
  n17 = OneMinus(None=ComponentMask(None=TextureCoordinate()) {g=true})
  n18 = ComponentMask(None=TextureCoordinate()) {r=true}
  n19 = Multiply(A=Power(Base=Sine(None=@n18) {period=2}) {constExponent=2.2})
  n20 = Multiply(
   A=LinearInterpolate(A=Constant() {r=0.2}, B=Constant() {r=1}, Alpha=Multiply(A=@n11.R, B=@n16.R))
   B=Multiply(A=Power(Base=@n17) {constExponent=2.2}, B=Clamp(None=@n19))
  )
  n21 = Power(Base=OneMinus(None=MaterialFunctionCall() {materialFunction=&LinearGradient}.VGradient)) {constExponent=3}
  n22 = DepthFade(FadeDistance=ScalarParameter() {defaultValue=100, parameterName=Depthfade})
  n23 = Constant() {r=1}
  Multiply(
   A=Multiply(A=Multiply(A=@n6.A, B=@n20), B=OneMinus(None=@n21))
   B=FeatureLevelSwitch(Default=@n22, ES2=@n23, ES3_1=@n23, SM4_REMOVED=@n22, SM5=@n22, SM6=@n22)
  )
 ]
}
