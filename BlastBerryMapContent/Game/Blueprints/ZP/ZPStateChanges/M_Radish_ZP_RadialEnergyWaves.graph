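# Unlit, two-sided translucent material for Niagara mesh particles, drawing radial energy waves for ZP state-change effects, colored by particle color.
# It builds three mask layers: noise-wobbled core airglow, energy waves distorted by two particle-randomized panning noises, and panning god rays.
# Their clamped sum is shaped by particle alpha, an axis-aligned Fresnel term and a generated band.
# Depth fade softens that opacity, except on mobile and low quality.
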Material /BlastBerryMapContent/Game/Blueprints/ZP/ZPStateChanges/M_Radish_ZP_RadialEnergyWaves
graph {
 properties {
  bUsedWithNiagaraMeshParticles=true
  blendMode=BLEND_Translucent
  pixelDepthOffsetMode=PDOM_Legacy
  shadingModel=MSM_Unlit
  twoSided=true
  usedShadingModels=Unlit
 }
 nodes [
  n1 = ParticleColor()
  MaterialFunctionCall(Emissive=@n1.RGB) {materialFunction=&MF_FX_EmissiveAdjustments}
  n2 = TextureCoordinate()
  n3 = Multiply(A=TextureCoordinate(), B=AppendVector(A=Constant() {r=2}, B=Constant() {r=1}))
  n4 = AppendVector(A=Constant() {r=0.1}, B=Constant() {r=0.1})
  n5 = Multiply(A=TextureSample(UVs=Panner(Coordinate=@n3, Speed=@n4)) {texture=/Game/Characters/Player/Male/Medium/Bodies/M_MED_Viper/Textures/Glitch/T_SoftNoise}.R) {constB=0.05}
  n6 = Constant()
  n7 = ScalarParameter() {defaultValue=-0.04, parameterName=CoreAirglow_Offset_Y}
  n8 = TextureSample(UVs=Add(A=Add(A=@n2, B=@n5), B=AppendVector(A=@n6, B=@n7))) {samplerSource=SSM_Clamp_WorldGroupSettings, texture=&/T_FX_Radish_ZP_AirglowMask}
  n9 = Power(Base=@n8.RGB, Exp=ScalarParameter() {defaultValue=1, parameterName=CoreAirglow_Power})
  n10 = ScalarParameter() {defaultValue=0.2, parameterName=CoreAirglow_Mult}
  NamedRerouteDeclaration(None=Clamp(None=Multiply(A=@n9, B=@n10))) {name=CoreAirglow, nodeColor=<rgba 1 0.7031 0 1>}
  n11 = TextureCoordinate()
  n12 = AppendVector(A=ScalarParameter() {defaultValue=10, parameterName=InnerNoise_Tile_X}, B=ScalarParameter() {parameterName=InnerNoise_Tile_Y})
  n13 = ScalarParameter() {parameterName=InnerNoise_Pan_X}
  n14 = ScalarParameter() {defaultValue=0.05, parameterName=InnerNoise_Pan_Y}
  n15 = TextureSample(UVs=Panner(Coordinate=Multiply(A=@n11, B=@n12), Speed=AppendVector(A=@n13, B=@n14))) {texture=&T_NoisePack_01}
  n16 = ScalarParameter() {defaultValue=1, parameterName=InnerNoise_Power}
  n17 = LinearInterpolate(
   A=Constant() {r=0.1}
   B=Constant() {r=1}
   Alpha=Multiply(A=Power(Base=@n15.R, Exp=@n16), B=ScalarParameter() {defaultValue=1, parameterName=InnerNoise_Mult})
  )
  n18 = ScalarParameter() {defaultValue=1, parameterName=Noise_Tile_X}
  n19 = ScalarParameter() {defaultValue=1, parameterName=Noise_Tile_Y}
  n20 = Panner(
   Coordinate=Multiply(A=TextureCoordinate(), B=AppendVector(A=@n18, B=@n19))
   Speed=AppendVector(A=ScalarParameter() {parameterName=Noise_Pan_X}, B=ScalarParameter() {parameterName=Noise_Pan_Y})
  )
  n21 = AppendVector(A=Constant(), B=ScalarParameter() {defaultValue=-0.025, parameterName=Noise_Offset_Y})
  n22 = Multiply(
   A=TextureCoordinate()
   B=AppendVector(
    A=ScalarParameter() {defaultValue=4, parameterName=Distortion_Tile_X}
    B=ScalarParameter() {defaultValue=0.1, parameterName=Distortion_Tile_Y}
   )
  )
  n23 = AppendVector(
   A=ScalarParameter() {defaultValue=0.01, parameterName=Distortion_Pan_X}
   B=ScalarParameter() {defaultValue=0.02, parameterName=Distortion_Pan_Y}
  )
  n24 = ParticleRandom()
  n25 = TextureSample(UVs=Add(A=Panner(Coordinate=@n22, Speed=@n23), B=AppendVector(A=@n24, B=@n24))) {samplerType=SAMPLERTYPE_Masks, texture=/Game/Weapons/WeaponSkins/Wraps/Textures/Mash/Mash_Noise}
  n26 = ScalarParameter() {defaultValue=2, parameterName=Distortion_Power}
  n27 = Multiply(
   A=TextureCoordinate()
   B=AppendVector(
    A=ScalarParameter() {defaultValue=2, parameterName=Distortion_Micro_Tile_X}
    B=ScalarParameter() {defaultValue=1, parameterName=Distortion_Micro_Tile_Y}
   )
  )
  n28 = AppendVector(A=ScalarParameter() {parameterName=Distortion_Micro_Pan_X}, B=ScalarParameter() {defaultValue=0.05, parameterName=Distortion_Micro_Pan_Y})
  n29 = ParticleRandom()
  n30 = TextureSample(UVs=Add(A=Panner(Coordinate=@n27, Speed=@n28), B=AppendVector(A=@n29, B=@n29))) {texture=&AtmosphericCloudNoise01}
  n31 = ScalarParameter() {defaultValue=1, parameterName=Distortion_Micro_Power}
  n32 = Add(
   A=Multiply(A=Power(Base=@n25.R, Exp=@n26), B=ScalarParameter() {defaultValue=0.3, parameterName=Distortion_Mult})
   B=Multiply(A=Power(Base=@n30.R, Exp=@n31), B=ScalarParameter() {defaultValue=0.05, parameterName=Distortion_Micro_Mult})
  )
  n33 = TextureSample(UVs=Add(A=Add(A=@n20, B=@n21), B=Clamp(None=@n32))) {samplerSource=SSM_Clamp_WorldGroupSettings, texture=&/T_FX_Radish_ZP_EnergyWaveMask}
  n34 = Multiply(
   A=Power(Base=@n33.RGB, Exp=ScalarParameter() {defaultValue=1, parameterName=Noise_Power})
   B=ScalarParameter() {defaultValue=1, parameterName=Noise_Mult}
  )
  NamedRerouteDeclaration(None=Multiply(A=Clamp(None=@n17), B=Clamp(None=@n34))) {name=EnergyWaves, nodeColor=<rgba 1 0 0.2344 1>}
  n35 = TextureCoordinate()
  n36 = AppendVector(A=ScalarParameter() {defaultValue=4, parameterName=GodRays_Tile_X}, B=Constant())
  n37 = ScalarParameter() {parameterName=GodRays_Pan_X}
  n38 = ScalarParameter() {defaultValue=0.05, parameterName=GodRays_Pan_Y}
  n39 = TextureSample(UVs=Panner(Coordinate=Multiply(A=@n35, B=@n36), Speed=AppendVector(A=@n37, B=@n38))) {texture=&T_NoisePack_01}
  n40 = Power(Base=ComponentMask(None=TextureCoordinate()) {g=true}) {constExponent=3}
  NamedRerouteDeclaration(None=Multiply(A=Multiply(A=@n39.R, B=@n40), B=ScalarParameter() {defaultValue=1, parameterName=GodRays_Mult})) {name=GodBeams, nodeColor=<rgba 0.9922 0 1 1>}
  n41 = Add(A=NamedRerouteUsage(), B=NamedRerouteUsage())
  n42 = Clamp(None=Add(A=Clamp(None=@n41), B=NamedRerouteUsage()) {constB=0})
  n43 = CrossProduct(A=VertexNormalWS(), B=VertexTangentWS())
  n44 = MaterialFunctionCall() {materialFunction=&CameraDirectionVector}
  n45 = Constant3Vector() {constant=<rgba 0 0 1 0>}
  n46 = LinearInterpolate(A=Constant() {r=1}, B=Constant() {r=1}, Alpha=Abs(None=DotProduct(A=@n44.Result, B=@n45)))
  n47 = Power(Base=OneMinus(None=MaterialFunctionCall(Axis=@n43, Exponent=@n46) {materialFunction=&AxisAlignedFresnel}.Fresnel)) {constExponent=1}
  n48 = Multiply(
   A=Multiply(A=Multiply(A=@n1.A, B=@n42), B=OneMinus(None=@n47))
   B=MaterialFunctionCall!(Width=Constant() {r=0.5}, Sharpness=Constant() {r=0.5}, Offset, Direction Switch, Compare, Input Coordinates) {materialFunction=/Engine/Functions/Engine_MaterialFunctions03/Procedurals/GeneratedBand}.Result
  )
  n49 = ScalarParameter() {defaultValue=100, parameterName=Fade distance}
  QualitySwitch(Default=ShadingPathSwitch(Default=DepthFade(Opacity=@n48, FadeDistance=@n49), Mobile=@n48), Low=@n48)
  TextureSample() {texture=&/T_FX_Radish_ZP_EnergyWave_ThinMask}
 ]
}
